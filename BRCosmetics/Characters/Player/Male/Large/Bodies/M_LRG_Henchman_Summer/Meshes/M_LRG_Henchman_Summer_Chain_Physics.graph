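# Physics asset for the two chain necklaces, one with a skull pendant, on the large male Summer Henchman character body.
# Each necklace is a three-link series of boxes constrained back to a small master capsule, with the longer one ending in the skull capsule.
# Joints have modest swing limits and very tight twist limits.
# An unconstrained spine_05 body combines capsules and an angled thin box as an upper-torso collision shape.
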PhysicsAsset /BRCosmetics/Characters/Player/Male/Large/Bodies/M_LRG_Henchman_Summer/Meshes/M_LRG_Henchman_Summer_Chain_Physics
bodies {
 Dyn_Master_Chain [{shapeType=Capsule, length=0.5, radius=0.5, rotation=<rot 0 -90 0>}]
 Dyn_Master_Short_Chain [{shapeType=Capsule, length=0.5, radius=0.5, rotation=<rot 0 -90 0>}]
 Dyn_Necklace_Chain1 [{shapeType=Box, extentX=5, extentY=1, extentZ=20, shapeName=Dyn_Necklace_Chain1_box, center=<-0.4943 -0.756 -0.0001>}]
 Dyn_Necklace_Chain2 [{shapeType=Box, extentX=5, extentY=1, extentZ=20, shapeName=Dyn_Necklace_Chain2_box, center=<-0.2371 -0.7559 -0.0001>}]
 Dyn_Necklace_Chain3 [{shapeType=Box, extentX=5, extentY=1, extentZ=20, shapeName=Dyn_Necklace_Chain3_box, center=<0.2486 -0.7561 -0.0001>}]
 Dyn_Necklace_Short_Chain1 [{shapeType=Box, extentX=10, extentY=2, extentZ=20, shapeName=Dyn_Necklace_Short_Chain1_box, center=<0 -0.5338 0>}]
 Dyn_Necklace_Short_Chain2 [{shapeType=Box, extentX=8, extentY=2, extentZ=15, shapeName=Dyn_Necklace_Short_Chain2_box, center=<1.2327 0 0>}]
 Dyn_Necklace_Short_Chain3 [{shapeType=Box, extentX=5, extentY=2, extentZ=10, shapeName=Dyn_Necklace_Short_Chain3_box, rotation=<rot 0 10.0005 0>}]
 Dyn_Skull [{shapeType=Capsule, length=1.5, radius=1.75, center=<2.4796 -0.3326 0>, rotation=<rot 90 0 -8.7821>}]
 spine_05 [
  {shapeType=Capsule, length=15.9718, radius=15, center=<-10.6628 2.4603 0>}
  {shapeType=Capsule, length=15, radius=15, center=<-2.686 -2.6424 0>, rotation=<rot 0 -29.9115 0>}
  {shapeType=Capsule, length=15, radius=15, center=<-7.3935 0.8717 0>}
  {shapeType=Box, extentX=1, extentY=30, extentZ=30, shapeName=spine_05_box, center=<3.4957 8.6828 0>, rotation=<rot 0 50.0002 0>}
 ]
}
constraints [
 {bone1Name=Dyn_Necklace_Chain1, bone2Name=Dyn_Master_Chain, swing1LimitDegrees=20, swing2LimitDegrees=20, twistLimitDegrees=3}
 {bone1Name=Dyn_Necklace_Chain2, bone2Name=Dyn_Necklace_Chain1, swing1LimitDegrees=20, swing2LimitDegrees=20, twistLimitDegrees=3}
 {bone1Name=Dyn_Necklace_Chain3, bone2Name=Dyn_Necklace_Chain2, swing1LimitDegrees=15, swing2LimitDegrees=10, twistLimitDegrees=3}
 {bone1Name=Dyn_Skull, bone2Name=Dyn_Necklace_Chain3, swing1LimitDegrees=10, swing2LimitDegrees=20, twistLimitDegrees=1}
 {bone1Name=Dyn_Necklace_Short_Chain1, bone2Name=Dyn_Master_Short_Chain, swing1LimitDegrees=20, swing2LimitDegrees=20, twistLimitDegrees=3}
 {bone1Name=Dyn_Necklace_Short_Chain2, bone2Name=Dyn_Necklace_Short_Chain1, swing1LimitDegrees=20, swing2LimitDegrees=20, twistLimitDegrees=3}
 {bone1Name=Dyn_Necklace_Short_Chain3, bone2Name=Dyn_Necklace_Short_Chain2, swing1LimitDegrees=15, swing2LimitDegrees=10, twistLimitDegrees=3}
]
