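# Physics asset for the left leg of the large male Summer Henchman body, setting up bone-driven physics for the shorts.
# Three thin capsules on the Dyn_Shorts_l1, l3 and l4 bones are each constrained to a wider capsule on Dyn_Master_Shorts_l.
# These joints have limited swing and almost no twist.
# A separate capsule on calf_l has no constraint and gives the lower leg a collision shape.
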PhysicsAsset /BRCosmetics/Characters/Player/Male/Large/Bodies/M_LRG_Henchman_Summer/Meshes/M_LRG_Henchman_Summer_Leg_L_Physics
bodies {
 Dyn_Master_Shorts_l [{shapeType=Capsule, length=15, radius=7.5, center=<8.5799 0 1.7328>, rotation=<rot 84.4205 0 0>}]
 Dyn_Shorts_l1 [{shapeType=Capsule, length=20, radius=1.5, center=<0.9884 -0.0001 2.0292>}]
 Dyn_Shorts_l3 [{shapeType=Capsule, length=20, radius=1.5, center=<0.9884 -0.0001 0.0006>}]
 Dyn_Shorts_l4 [{shapeType=Capsule, length=20, radius=1.5, center=<0.9884 -0.0001 0.0006>}]
 calf_l [{shapeType=Capsule, length=17.5, radius=4.5, center=<-15.1714 -2.7379 -0.9037>, rotation=<rot 90 0 0>}]
}
constraints [
 {bone1Name=Dyn_Shorts_l1, bone2Name=Dyn_Master_Shorts_l, swing1LimitDegrees=30, swing2LimitDegrees=45, twistLimitDegrees=1}
 {bone1Name=Dyn_Shorts_l3, bone2Name=Dyn_Master_Shorts_l, swing1LimitDegrees=30, swing2LimitDegrees=45, twistLimitDegrees=1}
 {bone1Name=Dyn_Shorts_l4, bone2Name=Dyn_Master_Shorts_l, swing1LimitDegrees=30, swing2LimitDegrees=45, twistLimitDegrees=1}
]
