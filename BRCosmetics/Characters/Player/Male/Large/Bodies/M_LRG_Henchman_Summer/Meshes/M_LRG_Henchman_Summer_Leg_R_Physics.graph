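# Physics asset for the right leg mesh of the large male Summer Henchman cosmetic body, set up for dynamic shorts bones.
# Three thin capsules on the shorts bones (r1, r3, r4) each constrain directly to a wider capsule on the master shorts bone, rather than forming a chain.
# The constraints allow moderate swing with almost no twist.
# An unconstrained calf capsule adds a body for the lower leg.
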PhysicsAsset /BRCosmetics/Characters/Player/Male/Large/Bodies/M_LRG_Henchman_Summer/Meshes/M_LRG_Henchman_Summer_Leg_R_Physics
bodies {
 Dyn_Master_Shorts_r [{shapeType=Capsule, length=14, radius=8.5, center=<8.5145 0 -1.933>, rotation=<rot 86.9257 -179.9999 -179.9999>}]
 Dyn_Shorts_r1 [{shapeType=Capsule, length=20, radius=1.5, center=<0.7971 -0.0001 -1.7126>}]
 Dyn_Shorts_r3 [{shapeType=Capsule, length=20, radius=1.5, center=<0.7971 -0.0001 0.0005>}]
 Dyn_Shorts_r4 [{shapeType=Capsule, length=20, radius=1.5, center=<0.7971 -0.0001 0.0005>}]
 calf_r [{shapeType=Capsule, length=17.5, radius=4.5, center=<15.4355 2.3189 0.2707>, rotation=<rot 90 0 0>}]
}
constraints [
 {bone1Name=Dyn_Shorts_r1, bone2Name=Dyn_Master_Shorts_r, swing1LimitDegrees=30, swing2LimitDegrees=45, twistLimitDegrees=1}
 {bone1Name=Dyn_Shorts_r3, bone2Name=Dyn_Master_Shorts_r, swing1LimitDegrees=30, swing2LimitDegrees=45, twistLimitDegrees=1}
 {bone1Name=Dyn_Shorts_r4, bone2Name=Dyn_Master_Shorts_r, swing1LimitDegrees=30, swing2LimitDegrees=45, twistLimitDegrees=1}
]
