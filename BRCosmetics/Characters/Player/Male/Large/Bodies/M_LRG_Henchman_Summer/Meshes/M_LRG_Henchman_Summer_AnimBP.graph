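# Post-process style AnimBP for the large male Summer Henchman body cosmetic, adding secondary motion to shorts and necklace chain.
# The parent's copied pose runs through RigidBody simulations for both legs and chain, additive bone corrections, then AnimDynamics on two shorts bones.
# Each update derives correction alphas from neck, clavicle and leg transform deltas.
# Crouch, sprint, jump, fall and Neck_Up bone chains are unconnected to the output.
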AnimBlueprint /BRCosmetics/Characters/Player/Male/Large/Bodies/M_LRG_Henchman_Summer/Meshes/M_LRG_Henchman_Summer_AnimBP
target_skeleton=&/M_LRG_Henchman_Summer_Skeleton
graphs [
 AnimGraph {nodes [
   n1 = Get bIsCrouching()
   n2 = "Transform (Modify) Bone - Bone: Dyn_Shorts_l2"!(ComponentPose, bAlphaBoolEnabled:bool=@n1, Translation="-8,0,0", Rotation="0,0,0", Scale="1,1,1") -> (Pose) {
    alphaInputType=Bool
    rotationMode=BMM_Additive
    rotationSpace=BCS_BoneSpace
    scaleMode=BMM_Additive
    scaleSpace=BCS_BoneSpace
    translationMode=BMM_Additive
    translationSpace=BCS_BoneSpace
    alphaBoolBlend {blendOption=Linear}
    boneToModify {boneName=Dyn_Shorts_l2}
    translation {x=-8}
   }
   n3 = "Transform (Modify) Bone - Bone: Dyn_Shorts_l3"!(ComponentPose=@n2, bAlphaBoolEnabled:bool=@n1, Translation="-5,0,0", Rotation="0,0,0", Scale="1,1,1") -> (Pose) {
    alphaInputType=Bool
    rotationMode=BMM_Additive
    rotationSpace=BCS_BoneSpace
    scaleMode=BMM_Additive
    scaleSpace=BCS_BoneSpace
    translationMode=BMM_Additive
    translationSpace=BCS_BoneSpace
    alphaBoolBlend {blendOption=Linear}
    boneToModify {boneName=Dyn_Shorts_l3}
    translation {x=-5}
   }
   n4 = "Transform (Modify) Bone - Bone: Dyn_Shorts_l4"!(ComponentPose=@n3, bAlphaBoolEnabled:bool=@n1, Translation="-5,0,0", Rotation="0,0,0", Scale="1,1,1") -> (Pose) {
    alphaInputType=Bool
    rotationMode=BMM_Additive
    rotationSpace=BCS_BoneSpace
    scaleMode=BMM_Additive
    scaleSpace=BCS_BoneSpace
    translationMode=BMM_Additive
    translationSpace=BCS_BoneSpace
    alphaBoolBlend {blendOption=Linear}
    boneToModify {boneName=Dyn_Shorts_l4}
    translation {x=-5}
   }
   n5 = "Transform (Modify) Bone - Bone: Dyn_Shorts_r2"!(ComponentPose=@n4, bAlphaBoolEnabled:bool=@n1, Translation="-7,0,0", Rotation="0,0,0", Scale="0.8,1,1") -> (Pose) {
    alphaInputType=Bool
    rotationMode=BMM_Additive
    rotationSpace=BCS_BoneSpace
    scaleMode=BMM_Additive
    scaleSpace=BCS_BoneSpace
    translationMode=BMM_Additive
    translationSpace=BCS_BoneSpace
    alphaBoolBlend {blendOption=Linear}
    boneToModify {boneName=Dyn_Shorts_r2}
    scale {x=0.8}
    translation {x=-7}
   }
   n6 = "Transform (Modify) Bone - Bone: Dyn_Shorts_r3"!(ComponentPose=@n5, bAlphaBoolEnabled:bool=@n1, Translation="-7,0,0", Rotation="0,0,0", Scale="0.8,1,1") -> (Pose) {
    alphaInputType=Bool
    rotationMode=BMM_Additive
    rotationSpace=BCS_BoneSpace
    scaleMode=BMM_Additive
    scaleSpace=BCS_BoneSpace
    translationMode=BMM_Additive
    translationSpace=BCS_BoneSpace
    alphaBoolBlend {blendOption=Linear}
    boneToModify {boneName=Dyn_Shorts_r3}
    scale {x=0.8}
    translation {x=-7}
   }
   n7 = "Transform (Modify) Bone - Bone: Dyn_Shorts_r1"!(ComponentPose=@n6, bAlphaBoolEnabled:bool=@n1, Translation="-5,0,0", Rotation="0,0,0", Scale="0.8,1,1") -> (Pose) {
    alphaInputType=Bool
    rotationMode=BMM_Additive
    rotationSpace=BCS_BoneSpace
    scaleMode=BMM_Additive
    scaleSpace=BCS_BoneSpace
    translationMode=BMM_Additive
    translationSpace=BCS_BoneSpace
    alphaBoolBlend {blendOption=Linear}
    boneToModify {boneName=Dyn_Shorts_r1}
    scale {x=0.8}
    translation {x=-5}
   }
   n8 = "Transform (Modify) Bone - Bone: Dyn_Shorts_r4"!(ComponentPose=@n7, bAlphaBoolEnabled:bool=@n1, Translation="-5,0,0", Rotation="0,0,0", Scale="0.8,1,1") -> (Pose) {
    alphaInputType=Bool
    rotationMode=BMM_Additive
    rotationSpace=BCS_BoneSpace
    scaleMode=BMM_Additive
    scaleSpace=BCS_BoneSpace
    translationMode=BMM_Additive
    translationSpace=BCS_BoneSpace
    alphaBoolBlend {blendOption=Linear}
    boneToModify {boneName=Dyn_Shorts_r4}
    scale {x=0.8}
    translation {x=-5}
   }
   n9 = Get bIsCrouchSprinting()
   n10 = "Transform (Modify) Bone - Bone: Dyn_Shorts_l2"(ComponentPose=@n8, bAlphaBoolEnabled=@n9, Translation="-8,0,0", Rotation="0,0,0", Scale="1,1,1") {
    alphaInputType=Bool
    rotationMode=BMM_Additive
    rotationSpace=BCS_BoneSpace
    scaleMode=BMM_Additive
    scaleSpace=BCS_BoneSpace
    translationMode=BMM_Additive
    translationSpace=BCS_BoneSpace
    alphaBoolBlend {blendOption=Linear}
    boneToModify {boneName=Dyn_Shorts_l2}
    translation {x=-8}
   }
   n11 = "Transform (Modify) Bone - Bone: Dyn_Shorts_l3"(ComponentPose=@n10, bAlphaBoolEnabled=@n9, Translation="-5,0,0", Rotation="0,0,0", Scale="1,1,1") {
    alphaInputType=Bool
    rotationMode=BMM_Additive
    rotationSpace=BCS_BoneSpace
    scaleMode=BMM_Additive
    scaleSpace=BCS_BoneSpace
    translationMode=BMM_Additive
    translationSpace=BCS_BoneSpace
    alphaBoolBlend {blendOption=Linear}
    boneToModify {boneName=Dyn_Shorts_l3}
    translation {x=-5}
   }
   n12 = "Transform (Modify) Bone - Bone: Dyn_Shorts_l4"(ComponentPose=@n11, bAlphaBoolEnabled=@n9, Translation="-5,0,0", Rotation="0,0,0", Scale="1,1,1") {
    alphaInputType=Bool
    rotationMode=BMM_Additive
    rotationSpace=BCS_BoneSpace
    scaleMode=BMM_Additive
    scaleSpace=BCS_BoneSpace
    translationMode=BMM_Additive
    translationSpace=BCS_BoneSpace
    alphaBoolBlend {blendOption=Linear}
    boneToModify {boneName=Dyn_Shorts_l4}
    translation {x=-5}
   }
   n13 = "Transform (Modify) Bone - Bone: Dyn_Shorts_r2"(ComponentPose=@n12, bAlphaBoolEnabled=@n9, Translation="-12,0,0", Rotation="0,0,0", Scale="1,1,1") {
    alphaInputType=Bool
    rotationMode=BMM_Additive
    rotationSpace=BCS_BoneSpace
    scaleMode=BMM_Additive
    scaleSpace=BCS_BoneSpace
    translationMode=BMM_Additive
    translationSpace=BCS_BoneSpace
    alphaBoolBlend {blendOption=Linear}
    boneToModify {boneName=Dyn_Shorts_r2}
    translation {x=-12}
   }
   n14 = "Transform (Modify) Bone - Bone: Dyn_Shorts_r3"(ComponentPose=@n13, bAlphaBoolEnabled=@n9, Translation="-5,0,0", Rotation="0,0,0", Scale="1,1,1") {
    alphaInputType=Bool
    rotationMode=BMM_Additive
    rotationSpace=BCS_BoneSpace
    scaleMode=BMM_Additive
    scaleSpace=BCS_BoneSpace
    translationMode=BMM_Additive
    translationSpace=BCS_BoneSpace
    alphaBoolBlend {blendOption=Linear}
    boneToModify {boneName=Dyn_Shorts_r3}
    translation {x=-5}
   }
   "Transform (Modify) Bone - Bone: Dyn_Shorts_r4"(ComponentPose=@n14, bAlphaBoolEnabled=@n9, Translation="-5,0,0", Rotation="0,0,0", Scale="1,1,1") {
    alphaInputType=Bool
    rotationMode=BMM_Additive
    rotationSpace=BCS_BoneSpace
    scaleMode=BMM_Additive
    scaleSpace=BCS_BoneSpace
    translationMode=BMM_Additive
    translationSpace=BCS_BoneSpace
    alphaBoolBlend {blendOption=Linear}
    boneToModify {boneName=Dyn_Shorts_r4}
    translation {x=-5}
   }
   n15 = Get bIsJumping()
   n16 = "Transform (Modify) Bone - Bone: Dyn_Master_Shorts_l"!(ComponentPose, bAlphaBoolEnabled:bool=@n15, Translation="-8,0,0", Rotation="0,0,0", Scale="1,1,1") -> (Pose) {
    alphaInputType=Bool
    rotationMode=BMM_Additive
    rotationSpace=BCS_BoneSpace
    scaleMode=BMM_Additive
    scaleSpace=BCS_BoneSpace
    translationMode=BMM_Additive
    translationSpace=BCS_BoneSpace
    alphaBoolBlend {blendOption=Linear}
    boneToModify {boneName=Dyn_Master_Shorts_l}
    translation {x=-8}
   }
   "Transform (Modify) Bone - Bone: Dyn_Master_Shorts_r"!(ComponentPose=@n16, bAlphaBoolEnabled:bool=@n15, Translation="-12,0,0", Rotation="0,0,0", Scale="1,1,1") -> (Pose) {
    alphaInputType=Bool
    rotationMode=BMM_Additive
    rotationSpace=BCS_BoneSpace
    scaleMode=BMM_Additive
    scaleSpace=BCS_BoneSpace
    translationMode=BMM_Additive
    translationSpace=BCS_BoneSpace
    alphaBoolBlend {blendInTime=1, blendOption=Linear, blendOutTime=1}
    boneToModify {boneName=Dyn_Master_Shorts_r}
    translation {x=-12}
   }
   n17 = Get bIsFalling()
   n18 = "Transform (Modify) Bone - Bone: Dyn_Shorts_l2"(bAlphaBoolEnabled=@n17, Translation="-8,0,0", Rotation="0,0,0", Scale="1,1,1") {
    alphaInputType=Bool
    rotationMode=BMM_Additive
    rotationSpace=BCS_BoneSpace
    scaleMode=BMM_Additive
    scaleSpace=BCS_BoneSpace
    translationMode=BMM_Additive
    translationSpace=BCS_BoneSpace
    alphaBoolBlend {blendOption=Linear}
    boneToModify {boneName=Dyn_Shorts_l2}
    translation {x=-8}
   }
   n19 = "Transform (Modify) Bone - Bone: Dyn_Shorts_l3"(ComponentPose=@n18, bAlphaBoolEnabled=@n17, Translation="-5,0,0", Rotation="0,0,0", Scale="1,1,1") {
    alphaInputType=Bool
    rotationMode=BMM_Additive
    rotationSpace=BCS_BoneSpace
    scaleMode=BMM_Additive
    scaleSpace=BCS_BoneSpace
    translationMode=BMM_Additive
    translationSpace=BCS_BoneSpace
    alphaBoolBlend {blendOption=Linear}
    boneToModify {boneName=Dyn_Shorts_l3}
    translation {x=-5}
   }
   n20 = "Transform (Modify) Bone - Bone: Dyn_Shorts_l4"(ComponentPose=@n19, bAlphaBoolEnabled=@n17, Translation="-5,0,0", Rotation="0,0,0", Scale="1,1,1") {
    alphaInputType=Bool
    rotationMode=BMM_Additive
    rotationSpace=BCS_BoneSpace
    scaleMode=BMM_Additive
    scaleSpace=BCS_BoneSpace
    translationMode=BMM_Additive
    translationSpace=BCS_BoneSpace
    alphaBoolBlend {blendOption=Linear}
    boneToModify {boneName=Dyn_Shorts_l4}
    translation {x=-5}
   }
   n21 = "Transform (Modify) Bone - Bone: Dyn_Shorts_r2"(ComponentPose=@n20, bAlphaBoolEnabled=@n17, Translation="-12,0,0", Rotation="0,0,0", Scale="1,1,1") {
    alphaInputType=Bool
    rotationMode=BMM_Additive
    rotationSpace=BCS_BoneSpace
    scaleMode=BMM_Additive
    scaleSpace=BCS_BoneSpace
    translationMode=BMM_Additive
    translationSpace=BCS_BoneSpace
    alphaBoolBlend {blendOption=Linear}
    boneToModify {boneName=Dyn_Shorts_r2}
    translation {x=-12}
   }
   n22 = "Transform (Modify) Bone - Bone: Dyn_Shorts_r3"(ComponentPose=@n21, bAlphaBoolEnabled=@n17, Translation="0,0,0", Rotation="0,0,0", Scale="1,1,1") {
    alphaInputType=Bool
    rotationMode=BMM_Additive
    rotationSpace=BCS_BoneSpace
    scaleMode=BMM_Additive
    scaleSpace=BCS_BoneSpace
    translationMode=BMM_Additive
    translationSpace=BCS_BoneSpace
    alphaBoolBlend {blendOption=Linear}
    boneToModify {boneName=Dyn_Shorts_r3}
   }
   "Transform (Modify) Bone - Bone: Dyn_Shorts_r4"(ComponentPose=@n22, bAlphaBoolEnabled=@n17, Translation="-5,0,0", Rotation="0,0,0", Scale="1,1,1") {
    alphaInputType=Bool
    rotationMode=BMM_Additive
    rotationSpace=BCS_BoneSpace
    scaleMode=BMM_Additive
    scaleSpace=BCS_BoneSpace
    translationMode=BMM_Additive
    translationSpace=BCS_BoneSpace
    alphaBoolBlend {blendOption=Linear}
    boneToModify {boneName=Dyn_Shorts_r4}
    translation {x=-5}
   }
   n23 = "Transform (Modify) Bone - Bone: Dyn_Shorts_r4"(bAlphaBoolEnabled=@n1, Translation="0,0,0", Rotation="0,20,0", Scale="1,1,1") {
    alphaInputType=Bool
    rotationMode=BMM_Additive
    rotationSpace=BCS_BoneSpace
    scaleMode=BMM_Additive
    scaleSpace=BCS_BoneSpace
    translationMode=BMM_Additive
    translationSpace=BCS_BoneSpace
    alphaBoolBlend {blendOption=Linear}
    boneToModify {boneName=Dyn_Shorts_r4}
    rotation {yaw=20}
   }
   n24 = "Transform (Modify) Bone - Bone: Dyn_Master_Shorts_l"(ComponentPose=@n23, bAlphaBoolEnabled=@n9, Translation="0,2,0", Rotation="0,0,0", Scale="0.9,0.9,0.9") {
    alphaInputType=Bool
    rotationMode=BMM_Additive
    rotationSpace=BCS_BoneSpace
    scaleMode=BMM_Additive
    scaleSpace=BCS_BoneSpace
    translationMode=BMM_Additive
    translationSpace=BCS_BoneSpace
    alphaBoolBlend {blendOption=Linear}
    boneToModify {boneName=Dyn_Master_Shorts_l}
    scale=<0.9 0.9 0.9>
    translation {y=2}
   }
   "Transform (Modify) Bone - Bone: Dyn_Master_Shorts_r"(ComponentPose=@n24, bAlphaBoolEnabled=@n9, Translation="-10,0,0", Rotation="0,0,0", Scale="0.5,1,1") {
    alphaInputType=Bool
    rotationMode=BMM_Additive
    rotationSpace=BCS_BoneSpace
    scaleMode=BMM_Additive
    scaleSpace=BCS_BoneSpace
    translationMode=BMM_Additive
    translationSpace=BCS_BoneSpace
    alphaBoolBlend {blendOption=Linear}
    boneToModify {boneName=Dyn_Master_Shorts_r}
    scale {x=0.5}
    translation {x=-10}
   }
   n25 = Get bIsJumping()
   n26 = "Transform (Modify) Bone - Bone: Dyn_Shorts_l2"(bAlphaBoolEnabled=@n25, Translation="-20,0,0", Rotation="0,0,0", Scale="1,1,1") {
    alphaInputType=Bool
    rotationMode=BMM_Additive
    rotationSpace=BCS_BoneSpace
    scaleMode=BMM_Additive
    scaleSpace=BCS_BoneSpace
    translationMode=BMM_Additive
    translationSpace=BCS_BoneSpace
    alphaBoolBlend {blendInTime=5, blendOption=Linear, blendOutTime=5}
    boneToModify {boneName=Dyn_Shorts_l2}
    translation {x=-20}
   }
   n27 = "Transform (Modify) Bone - Bone: Dyn_Shorts_l3"(ComponentPose=@n26, bAlphaBoolEnabled=@n25, Translation="-5,0,0", Rotation="0,0,0", Scale="1,1,1") {
    alphaInputType=Bool
    rotationMode=BMM_Additive
    rotationSpace=BCS_BoneSpace
    scaleMode=BMM_Additive
    scaleSpace=BCS_BoneSpace
    translationMode=BMM_Additive
    translationSpace=BCS_BoneSpace
    alphaBoolBlend {blendInTime=5, blendOption=Linear, blendOutTime=5}
    boneToModify {boneName=Dyn_Shorts_l3}
    translation {x=-5}
   }
   n28 = "Transform (Modify) Bone - Bone: Dyn_Shorts_l4"(ComponentPose=@n27, bAlphaBoolEnabled=@n25, Translation="-5,0,0", Rotation="0,0,0", Scale="1,1,1") {
    alphaInputType=Bool
    rotationMode=BMM_Additive
    rotationSpace=BCS_BoneSpace
    scaleMode=BMM_Additive
    scaleSpace=BCS_BoneSpace
    translationMode=BMM_Additive
    translationSpace=BCS_BoneSpace
    alphaBoolBlend {blendInTime=5, blendOption=Linear, blendOutTime=5}
    boneToModify {boneName=Dyn_Shorts_l4}
    translation {x=-5}
   }
   n29 = "Transform (Modify) Bone - Bone: Dyn_Shorts_r2"(ComponentPose=@n28, bAlphaBoolEnabled=@n25, Translation="-7,0,0", Rotation="0,0,0", Scale="0.8,1,1") {
    alphaInputType=Bool
    rotationMode=BMM_Additive
    rotationSpace=BCS_BoneSpace
    scaleMode=BMM_Additive
    scaleSpace=BCS_BoneSpace
    translationMode=BMM_Additive
    translationSpace=BCS_BoneSpace
    alphaBoolBlend {blendInTime=5, blendOption=Linear, blendOutTime=5}
    boneToModify {boneName=Dyn_Shorts_r2}
    scale {x=0.8}
    translation {x=-7}
   }
   n30 = "Transform (Modify) Bone - Bone: Dyn_Shorts_r3"(ComponentPose=@n29, bAlphaBoolEnabled=@n25, Translation="-7,0,0", Rotation="0,0,0", Scale="0.8,1,1") {
    alphaInputType=Bool
    rotationMode=BMM_Additive
    rotationSpace=BCS_BoneSpace
    scaleMode=BMM_Additive
    scaleSpace=BCS_BoneSpace
    translationMode=BMM_Additive
    translationSpace=BCS_BoneSpace
    alphaBoolBlend {blendInTime=5, blendOption=Linear, blendOutTime=5}
    boneToModify {boneName=Dyn_Shorts_r3}
    scale {x=0.8}
    translation {x=-7}
   }
   n31 = "Transform (Modify) Bone - Bone: Dyn_Shorts_r1"(ComponentPose=@n30, bAlphaBoolEnabled=@n25, Translation="-5,0,0", Rotation="0,0,0", Scale="0.8,1,1") {
    alphaInputType=Bool
    rotationMode=BMM_Additive
    rotationSpace=BCS_BoneSpace
    scaleMode=BMM_Additive
    scaleSpace=BCS_BoneSpace
    translationMode=BMM_Additive
    translationSpace=BCS_BoneSpace
    alphaBoolBlend {blendInTime=5, blendOption=Linear, blendOutTime=5}
    boneToModify {boneName=Dyn_Shorts_r1}
    scale {x=0.8}
    translation {x=-5}
   }
   n32 = "Transform (Modify) Bone - Bone: Dyn_Shorts_r4"(ComponentPose=@n31, bAlphaBoolEnabled=@n25, Translation="-5,0,0", Rotation="0,0,0", Scale="0.8,1,1") {
    alphaInputType=Bool
    rotationMode=BMM_Additive
    rotationSpace=BCS_BoneSpace
    scaleMode=BMM_Additive
    scaleSpace=BCS_BoneSpace
    translationMode=BMM_Additive
    translationSpace=BCS_BoneSpace
    alphaBoolBlend {blendInTime=5, blendOption=Linear, blendOutTime=5}
    boneToModify {boneName=Dyn_Shorts_r4}
    scale {x=0.8}
    translation {x=-5}
   }
   n33 = Get bIsFalling()
   n34 = "Transform (Modify) Bone - Bone: Dyn_Shorts_l2"(ComponentPose=@n32, bAlphaBoolEnabled=@n33, Translation="-8,0,0", Rotation="0,0,0", Scale="1,1,1") {
    alphaInputType=Bool
    rotationMode=BMM_Additive
    rotationSpace=BCS_BoneSpace
    scaleMode=BMM_Additive
    scaleSpace=BCS_BoneSpace
    translationMode=BMM_Additive
    translationSpace=BCS_BoneSpace
    alphaBoolBlend {blendInTime=1, blendOption=Linear, blendOutTime=0.5}
    boneToModify {boneName=Dyn_Shorts_l2}
    translation {x=-8}
   }
   n35 = "Transform (Modify) Bone - Bone: Dyn_Shorts_l3"(ComponentPose=@n34, bAlphaBoolEnabled=@n33, Translation="-5,0,0", Rotation="0,0,0", Scale="1,1,1") {
    alphaInputType=Bool
    rotationMode=BMM_Additive
    rotationSpace=BCS_BoneSpace
    scaleMode=BMM_Additive
    scaleSpace=BCS_BoneSpace
    translationMode=BMM_Additive
    translationSpace=BCS_BoneSpace
    alphaBoolBlend {blendInTime=1, blendOption=Linear, blendOutTime=0.5}
    boneToModify {boneName=Dyn_Shorts_l3}
    translation {x=-5}
   }
   n36 = "Transform (Modify) Bone - Bone: Dyn_Shorts_l4"(ComponentPose=@n35, bAlphaBoolEnabled=@n33, Translation="-5,0,0", Rotation="0,0,0", Scale="1,1,1") {
    alphaInputType=Bool
    rotationMode=BMM_Additive
    rotationSpace=BCS_BoneSpace
    scaleMode=BMM_Additive
    scaleSpace=BCS_BoneSpace
    translationMode=BMM_Additive
    translationSpace=BCS_BoneSpace
    alphaBoolBlend {blendInTime=1, blendOption=Linear, blendOutTime=0.5}
    boneToModify {boneName=Dyn_Shorts_l4}
    translation {x=-5}
   }
   n37 = "Transform (Modify) Bone - Bone: Dyn_Shorts_r2"(ComponentPose=@n36, bAlphaBoolEnabled=@n33, Translation="-7,0,0", Rotation="0,0,0", Scale="0.8,1,1") {
    alphaInputType=Bool
    rotationMode=BMM_Additive
    rotationSpace=BCS_BoneSpace
    scaleMode=BMM_Additive
    scaleSpace=BCS_BoneSpace
    translationMode=BMM_Additive
    translationSpace=BCS_BoneSpace
    alphaBoolBlend {blendInTime=1, blendOption=Linear, blendOutTime=0.5}
    boneToModify {boneName=Dyn_Shorts_r2}
    scale {x=0.8}
    translation {x=-7}
   }
   n38 = "Transform (Modify) Bone - Bone: Dyn_Shorts_r3"(ComponentPose=@n37, bAlphaBoolEnabled=@n33, Translation="-7,0,0", Rotation="0,0,0", Scale="0.8,1,1") {
    alphaInputType=Bool
    rotationMode=BMM_Additive
    rotationSpace=BCS_BoneSpace
    scaleMode=BMM_Additive
    scaleSpace=BCS_BoneSpace
    translationMode=BMM_Additive
    translationSpace=BCS_BoneSpace
    alphaBoolBlend {blendInTime=1, blendOption=Linear, blendOutTime=0.5}
    boneToModify {boneName=Dyn_Shorts_r3}
    scale {x=0.8}
    translation {x=-7}
   }
   n39 = "Transform (Modify) Bone - Bone: Dyn_Shorts_r1"(ComponentPose=@n38, bAlphaBoolEnabled=@n33, Translation="-5,0,0", Rotation="0,0,0", Scale="0.8,1,1") {
    alphaInputType=Bool
    rotationMode=BMM_Additive
    rotationSpace=BCS_BoneSpace
    scaleMode=BMM_Additive
    scaleSpace=BCS_BoneSpace
    translationMode=BMM_Additive
    translationSpace=BCS_BoneSpace
    alphaBoolBlend {blendInTime=1, blendOption=Linear, blendOutTime=0.5}
    boneToModify {boneName=Dyn_Shorts_r1}
    scale {x=0.8}
    translation {x=-5}
   }
   "Transform (Modify) Bone - Bone: Dyn_Shorts_r4"(ComponentPose=@n39, bAlphaBoolEnabled=@n33, Translation="-5,0,0", Rotation="0,0,0", Scale="0.8,1,1") {
    alphaInputType=Bool
    rotationMode=BMM_Additive
    rotationSpace=BCS_BoneSpace
    scaleMode=BMM_Additive
    scaleSpace=BCS_BoneSpace
    translationMode=BMM_Additive
    translationSpace=BCS_BoneSpace
    alphaBoolBlend {blendInTime=1, blendOption=Linear, blendOutTime=0.5}
    boneToModify {boneName=Dyn_Shorts_r4}
    scale {x=0.8}
    translation {x=-5}
   }
   n40 = "Transform (Modify) Bone - Bone: Dyn_Master_Shorts_l"(bAlphaBoolEnabled=@n17, Translation="-8,0,0", Rotation="0,0,0", Scale="1,1,1") {
    alphaInputType=Bool
    rotationMode=BMM_Additive
    rotationSpace=BCS_BoneSpace
    scaleMode=BMM_Additive
    scaleSpace=BCS_BoneSpace
    translationMode=BMM_Additive
    translationSpace=BCS_BoneSpace
    alphaBoolBlend {blendOption=Linear}
    boneToModify {boneName=Dyn_Master_Shorts_l}
    translation {x=-8}
   }
   "Transform (Modify) Bone - Bone: Dyn_Master_Shorts_r"(ComponentPose=@n40, bAlphaBoolEnabled=@n17, Translation="-12,0,0", Rotation="0,0,0", Scale="1,1,1") {
    alphaInputType=Bool
    rotationMode=BMM_Additive
    rotationSpace=BCS_BoneSpace
    scaleMode=BMM_Additive
    scaleSpace=BCS_BoneSpace
    translationMode=BMM_Additive
    translationSpace=BCS_BoneSpace
    alphaBoolBlend {blendInTime=1, blendOption=Linear, blendOutTime=1}
    boneToModify {boneName=Dyn_Master_Shorts_r}
    translation {x=-12}
   }
   n41 = "Transform (Modify) Bone - Bone: Dyn_Shorts_l2"(bAlphaBoolEnabled=@n15, Translation="-8,0,0", Rotation="0,0,0", Scale="1,1,1") {
    alphaInputType=Bool
    rotationMode=BMM_Additive
    rotationSpace=BCS_BoneSpace
    scaleMode=BMM_Additive
    scaleSpace=BCS_BoneSpace
    translationMode=BMM_Additive
    translationSpace=BCS_BoneSpace
    alphaBoolBlend {blendOption=Linear}
    boneToModify {boneName=Dyn_Shorts_l2}
    translation {x=-8}
   }
   n42 = "Transform (Modify) Bone - Bone: Dyn_Shorts_l3"(ComponentPose=@n41, bAlphaBoolEnabled=@n15, Translation="-5,0,0", Rotation="0,0,0", Scale="1,1,1") {
    alphaInputType=Bool
    rotationMode=BMM_Additive
    rotationSpace=BCS_BoneSpace
    scaleMode=BMM_Additive
    scaleSpace=BCS_BoneSpace
    translationMode=BMM_Additive
    translationSpace=BCS_BoneSpace
    alphaBoolBlend {blendOption=Linear}
    boneToModify {boneName=Dyn_Shorts_l3}
    translation {x=-5}
   }
   n43 = "Transform (Modify) Bone - Bone: Dyn_Shorts_l4"(ComponentPose=@n42, bAlphaBoolEnabled=@n15, Translation="-5,0,0", Rotation="0,0,0", Scale="1,1,1") {
    alphaInputType=Bool
    rotationMode=BMM_Additive
    rotationSpace=BCS_BoneSpace
    scaleMode=BMM_Additive
    scaleSpace=BCS_BoneSpace
    translationMode=BMM_Additive
    translationSpace=BCS_BoneSpace
    alphaBoolBlend {blendOption=Linear}
    boneToModify {boneName=Dyn_Shorts_l4}
    translation {x=-5}
   }
   n44 = "Transform (Modify) Bone - Bone: Dyn_Shorts_r2"(ComponentPose=@n43, bAlphaBoolEnabled=@n15, Translation="-8,0,0", Rotation="0,0,0", Scale="1,1,1") {
    alphaInputType=Bool
    rotationMode=BMM_Additive
    rotationSpace=BCS_BoneSpace
    scaleMode=BMM_Additive
    scaleSpace=BCS_BoneSpace
    translationMode=BMM_Additive
    translationSpace=BCS_BoneSpace
    alphaBoolBlend {blendInTime=1, blendOption=Linear, blendOutTime=1}
    boneToModify {boneName=Dyn_Shorts_r2}
    translation {x=-8}
   }
   n45 = "Transform (Modify) Bone - Bone: Dyn_Shorts_r3"(ComponentPose=@n44, bAlphaBoolEnabled=@n15, Translation="-5,0,0", Rotation="0,0,0", Scale="1,1,1") {
    alphaInputType=Bool
    rotationMode=BMM_Additive
    rotationSpace=BCS_BoneSpace
    scaleMode=BMM_Additive
    scaleSpace=BCS_BoneSpace
    translationMode=BMM_Additive
    translationSpace=BCS_BoneSpace
    alphaBoolBlend {blendOption=Linear}
    boneToModify {boneName=Dyn_Shorts_r3}
    translation {x=-5}
   }
   "Transform (Modify) Bone - Bone: Dyn_Shorts_r1"(ComponentPose=@n45, bAlphaBoolEnabled=@n15, Translation="-8,0,0", Rotation="0,0,0", Scale="1,1,1") {
    alphaInputType=Bool
    rotationMode=BMM_Additive
    rotationSpace=BCS_BoneSpace
    scaleMode=BMM_Additive
    scaleSpace=BCS_BoneSpace
    translationMode=BMM_Additive
    translationSpace=BCS_BoneSpace
    alphaBoolBlend {blendOption=Linear}
    boneToModify {boneName=Dyn_Shorts_r1}
    translation {x=-8}
   }
   n46 = "Transform (Modify) Bone - Bone: Dyn_Master_Shorts_l"(bAlphaBoolEnabled=@n1, Translation="-2,0,0", Rotation="0.000,0.000,10", Scale="0.9,0.9,0.9") {
    alphaInputType=Bool
    rotationMode=BMM_Additive
    rotationSpace=BCS_BoneSpace
    scaleMode=BMM_Additive
    scaleSpace=BCS_BoneSpace
    translationMode=BMM_Additive
    translationSpace=BCS_BoneSpace
    alphaBoolBlend {blendOption=Linear}
    boneToModify {boneName=Dyn_Master_Shorts_l}
    rotation {roll=10}
    scale=<0.9 0.9 0.9>
    translation {x=-2}
   }
   "Transform (Modify) Bone - Bone: Dyn_Master_Shorts_r"(ComponentPose=@n46, bAlphaBoolEnabled=@n1, Translation="-8,0,0", Rotation="0,0,0", Scale="1,1,1") {
    alphaInputType=Bool
    rotationMode=BMM_Additive
    rotationSpace=BCS_BoneSpace
    scaleMode=BMM_Additive
    scaleSpace=BCS_BoneSpace
    translationMode=BMM_Additive
    translationSpace=BCS_BoneSpace
    alphaBoolBlend {blendOption=Linear}
    boneToModify {boneName=Dyn_Master_Shorts_r}
    translation {x=-8}
   }
   n47 = Get Neck_Up() -> (Neck_Up:double)
   n48 = "Transform (Modify) Bone - Bone: Dyn_Master_Short_Chain"(Alpha=@n47, Translation="-2,-2,0", Rotation="20,-20,0", Scale="1,1,1") {
    rotationMode=BMM_Additive
    rotationSpace=BCS_BoneSpace
    scaleMode=BMM_Additive
    scaleSpace=BCS_BoneSpace
    translationMode=BMM_Additive
    translationSpace=BCS_BoneSpace
    alphaBoolBlend {blendInTime=1, blendOption=Linear, blendOutTime=1}
    boneToModify {boneName=Dyn_Master_Short_Chain}
    rotation {pitch=20, yaw=-20}
    translation=<-2 -2>
   }
   "Transform (Modify) Bone - Bone: Dyn_Master_Chain"(ComponentPose=@n48, Alpha=@n47, Translation="0,0,0", Rotation="0.000,-20,0.000", Scale="1,1,1") {
    rotationMode=BMM_Additive
    rotationSpace=BCS_BoneSpace
    scaleMode=BMM_Additive
    scaleSpace=BCS_BoneSpace
    translationMode=BMM_Additive
    translationSpace=BCS_BoneSpace
    alphaBoolBlend {blendInTime=1, blendOption=Linear, blendOutTime=1}
    boneToModify {boneName=Dyn_Master_Chain}
    rotation {yaw=-20}
   }
   n49 = Get bIsSprinting()
   n50 = "Transform (Modify) Bone - Bone: Dyn_Shorts_l2"(bAlphaBoolEnabled=@n49, Translation="-8,0,0", Rotation="0,0,0", Scale="1,1,1") {
    alphaInputType=Bool
    rotationMode=BMM_Additive
    rotationSpace=BCS_BoneSpace
    scaleMode=BMM_Additive
    scaleSpace=BCS_BoneSpace
    translationMode=BMM_Additive
    translationSpace=BCS_BoneSpace
    alphaBoolBlend {blendInTime=1, blendOption=Linear, blendOutTime=1}
    boneToModify {boneName=Dyn_Shorts_l2}
    translation {x=-8}
   }
   n51 = "Transform (Modify) Bone - Bone: Dyn_Shorts_l3"(ComponentPose=@n50, bAlphaBoolEnabled=@n49, Translation="-5,0,0", Rotation="0,0,0", Scale="1,1,1") {
    alphaInputType=Bool
    rotationMode=BMM_Additive
    rotationSpace=BCS_BoneSpace
    scaleMode=BMM_Additive
    scaleSpace=BCS_BoneSpace
    translationMode=BMM_Additive
    translationSpace=BCS_BoneSpace
    alphaBoolBlend {blendInTime=1, blendOption=Linear, blendOutTime=1}
    boneToModify {boneName=Dyn_Shorts_l3}
    translation {x=-5}
   }
   n52 = "Transform (Modify) Bone - Bone: Dyn_Shorts_l4"(ComponentPose=@n51, bAlphaBoolEnabled=@n49, Translation="-5,0,0", Rotation="0,0,0", Scale="1,1,1") {
    alphaInputType=Bool
    rotationMode=BMM_Additive
    rotationSpace=BCS_BoneSpace
    scaleMode=BMM_Additive
    scaleSpace=BCS_BoneSpace
    translationMode=BMM_Additive
    translationSpace=BCS_BoneSpace
    alphaBoolBlend {blendInTime=1, blendOption=Linear, blendOutTime=1}
    boneToModify {boneName=Dyn_Shorts_l4}
    translation {x=-5}
   }
   n53 = "Transform (Modify) Bone - Bone: Dyn_Shorts_r2"(ComponentPose=@n52, bAlphaBoolEnabled=@n49, Translation="-7,0,0", Rotation="0,0,0", Scale="0.8,1,1") {
    alphaInputType=Bool
    rotationMode=BMM_Additive
    rotationSpace=BCS_BoneSpace
    scaleMode=BMM_Additive
    scaleSpace=BCS_BoneSpace
    translationMode=BMM_Additive
    translationSpace=BCS_BoneSpace
    alphaBoolBlend {blendInTime=1, blendOption=Linear, blendOutTime=1}
    boneToModify {boneName=Dyn_Shorts_r2}
    scale {x=0.8}
    translation {x=-7}
   }
   n54 = "Transform (Modify) Bone - Bone: Dyn_Shorts_r3"(ComponentPose=@n53, bAlphaBoolEnabled=@n49, Translation="-7,0,0", Rotation="0,0,0", Scale="0.8,1,1") {
    alphaInputType=Bool
    rotationMode=BMM_Additive
    rotationSpace=BCS_BoneSpace
    scaleMode=BMM_Additive
    scaleSpace=BCS_BoneSpace
    translationMode=BMM_Additive
    translationSpace=BCS_BoneSpace
    alphaBoolBlend {blendInTime=1, blendOption=Linear, blendOutTime=1}
    boneToModify {boneName=Dyn_Shorts_r3}
    scale {x=0.8}
    translation {x=-7}
   }
   n55 = "Transform (Modify) Bone - Bone: Dyn_Shorts_r1"(ComponentPose=@n54, bAlphaBoolEnabled=@n49, Translation="-5,0,0", Rotation="0,0,0", Scale="0.8,1,1") {
    alphaInputType=Bool
    rotationMode=BMM_Additive
    rotationSpace=BCS_BoneSpace
    scaleMode=BMM_Additive
    scaleSpace=BCS_BoneSpace
    translationMode=BMM_Additive
    translationSpace=BCS_BoneSpace
    alphaBoolBlend {blendInTime=1, blendOption=Linear, blendOutTime=1}
    boneToModify {boneName=Dyn_Shorts_r1}
    scale {x=0.8}
    translation {x=-5}
   }
   "Transform (Modify) Bone - Bone: Dyn_Shorts_r4"(ComponentPose=@n55, bAlphaBoolEnabled=@n49, Translation="-5,0,0", Rotation="0,0,0", Scale="0.8,1,1") {
    alphaInputType=Bool
    rotationMode=BMM_Additive
    rotationSpace=BCS_BoneSpace
    scaleMode=BMM_Additive
    scaleSpace=BCS_BoneSpace
    translationMode=BMM_Additive
    translationSpace=BCS_BoneSpace
    alphaBoolBlend {blendInTime=1, blendOption=Linear, blendOutTime=1}
    boneToModify {boneName=Dyn_Shorts_r4}
    scale {x=0.8}
    translation {x=-5}
   }
   n56 = RigidBody(ComponentPose=Local To Component(LocalPose=Copy Pose From Mesh() {bUseAttachedParent=true})) {
    bClampLinearTranslationLimitToRefPose=true
    bOverrideWorldGravity=true
    overridePhysicsAsset=&/M_LRG_Henchman_Summer_Leg_L_Physics
    simulationSpace=BaseBoneSpace
    alphaBoolBlend {blendOption=Linear}
    baseBoneRef {boneName=thigh_l}
    componentAppliedLinearAccClamp=<5000 5000 5000>
    componentLinearAccScale=<0.1 1 0.1>
    componentLinearVelScale=<0.1 1 0.1>
    overrideWorldGravity {z=-980}
   }
   n57 = RigidBody(ComponentPose=@n56) {
    bClampLinearTranslationLimitToRefPose=true
    bOverrideWorldGravity=true
    overridePhysicsAsset=&/M_LRG_Henchman_Summer_Leg_R_Physics
    simulationSpace=BaseBoneSpace
    alphaBoolBlend {blendOption=Linear}
    baseBoneRef {boneName=thigh_r}
    componentAppliedLinearAccClamp=<5000 5000 5000>
    componentLinearAccScale=<0.1 1 0.1>
    componentLinearVelScale=<0.1 1 0.1>
    overrideWorldGravity {z=-980}
   }
   n58 = RigidBody(ComponentPose=@n57) {
    bClampLinearTranslationLimitToRefPose=true
    bOverrideWorldGravity=true
    overridePhysicsAsset=&/M_LRG_Henchman_Summer_Chain_Physics
    simulationSpace=BaseBoneSpace
    alphaBoolBlend {blendOption=Linear}
    baseBoneRef {boneName=spine_03}
    componentAppliedLinearAccClamp=<6000 6000 6000>
    componentLinearAccScale=<0.25 0.25 0.25>
    componentLinearVelScale=<0.25 0.25 0.25>
    overrideWorldGravity {z=-980}
   }
   n59 = "Transform (Modify) Bone - Bone: Dyn_Master_Chain"(
    ComponentPose=@n58
    Alpha=Get Shoulder_SkullNecklace() -> (Shoulder_SkullNecklace:double)
    Translation="0,0,0"
    Rotation="-10,-11,-10"
    Scale="1,0.9,0.9"
   ) {
    rotationMode=BMM_Additive
    rotationSpace=BCS_BoneSpace
    scaleMode=BMM_Additive
    scaleSpace=BCS_BoneSpace
    translationMode=BMM_Additive
    translationSpace=BCS_BoneSpace
    alphaBoolBlend {blendInTime=1, blendOption=Linear, blendOutTime=1}
    alphaScaleBiasClamp {bInterpResult=true}
    boneToModify {boneName=Dyn_Master_Chain}
    rotation=<rot -10 -11 -10>
    scale {y=0.9, z=0.9}
   }
   n60 = "Transform (Modify) Bone - Bone: Dyn_Master_Chain"(ComponentPose=@n59, Alpha=Get Front end head() -> (Front end head:double), Translation="0,0,0", Rotation="0,-11,-10", Scale="1,0.95,0.95") {
    rotationMode=BMM_Additive
    rotationSpace=BCS_BoneSpace
    scaleMode=BMM_Additive
    scaleSpace=BCS_BoneSpace
    translationMode=BMM_Additive
    translationSpace=BCS_BoneSpace
    alphaBoolBlend {blendInTime=1, blendOption=Linear, blendOutTime=1}
    alphaScaleBiasClamp {bInterpResult=true}
    boneToModify {boneName=Dyn_Master_Chain}
    rotation {roll=-10, yaw=-11}
    scale {y=0.95, z=0.95}
   }
   n61 = Get Shorts_LeftLeg() -> (Shorts_LeftLeg:double)
   n62 = "Transform (Modify) Bone - Bone: Dyn_Shorts_l2"!(ComponentPose=@n60, Alpha:float=@n61, Translation="-8,0,0", Rotation="0,0,0", Scale="1,1,1") -> (Pose) {
    rotationMode=BMM_Additive
    rotationSpace=BCS_BoneSpace
    scaleMode=BMM_Additive
    scaleSpace=BCS_BoneSpace
    translationMode=BMM_Additive
    translationSpace=BCS_BoneSpace
    alphaBoolBlend {blendInTime=1, blendOption=Linear, blendOutTime=1}
    alphaScaleBiasClamp {bInterpResult=true, interpSpeedDecreasing=6}
    boneToModify {boneName=Dyn_Shorts_l2}
    translation {x=-8}
   }
   n63 = "Transform (Modify) Bone - Bone: Dyn_Shorts_l3"!(ComponentPose=@n62, Alpha:float=@n61, Translation="-5,0,0", Rotation="0,0,0", Scale="1,1,1") -> (Pose) {
    rotationMode=BMM_Additive
    rotationSpace=BCS_BoneSpace
    scaleMode=BMM_Additive
    scaleSpace=BCS_BoneSpace
    translationMode=BMM_Additive
    translationSpace=BCS_BoneSpace
    alphaBoolBlend {blendInTime=1, blendOption=Linear, blendOutTime=1}
    alphaScaleBiasClamp {bInterpResult=true, interpSpeedDecreasing=6}
    boneToModify {boneName=Dyn_Shorts_l3}
    translation {x=-5}
   }
   n64 = "Transform (Modify) Bone - Bone: Dyn_Shorts_l4"!(ComponentPose=@n63, Alpha:float=@n61, Translation="-5,0,0", Rotation="0,0,0", Scale="1,1,1") -> (Pose) {
    rotationMode=BMM_Additive
    rotationSpace=BCS_BoneSpace
    scaleMode=BMM_Additive
    scaleSpace=BCS_BoneSpace
    translationMode=BMM_Additive
    translationSpace=BCS_BoneSpace
    alphaBoolBlend {blendInTime=1, blendOption=Linear, blendOutTime=1}
    alphaScaleBiasClamp {bInterpResult=true, interpSpeedDecreasing=6}
    boneToModify {boneName=Dyn_Shorts_l4}
    translation {x=-5}
   }
   n65 = Get Shorts_RightLeg() -> (Shorts_RightLeg:double)
   n66 = "Transform (Modify) Bone - Bone: Dyn_Shorts_r2"!(ComponentPose=@n64, Alpha:float=@n65, Translation="-10,0,0", Rotation="-0.217,50.001,-0.231", Scale="0.800,0.500,1.000") -> (Pose) {
    rotationMode=BMM_Additive
    rotationSpace=BCS_BoneSpace
    scaleMode=BMM_Additive
    scaleSpace=BCS_BoneSpace
    translationMode=BMM_Additive
    translationSpace=BCS_BoneSpace
    alphaBoolBlend {blendInTime=1, blendOption=Linear, blendOutTime=1}
    alphaScaleBiasClamp {bInterpResult=true}
    boneToModify {boneName=Dyn_Shorts_r2}
    rotation=<rot -0.217 50.001 -0.231>
    scale=<0.8 0.5>
    translation {x=-10}
   }
   n67 = "Transform (Modify) Bone - Bone: Dyn_Shorts_r3"!(ComponentPose=@n66, Alpha:float=@n65, Translation="-7,0,0", Rotation="0,0,0", Scale="0.8,1,1") -> (Pose) {
    rotationMode=BMM_Additive
    rotationSpace=BCS_BoneSpace
    scaleMode=BMM_Additive
    scaleSpace=BCS_BoneSpace
    translationMode=BMM_Additive
    translationSpace=BCS_BoneSpace
    alphaBoolBlend {blendInTime=1, blendOption=Linear, blendOutTime=1}
    alphaScaleBiasClamp {bInterpResult=true, interpSpeedDecreasing=6}
    boneToModify {boneName=Dyn_Shorts_r3}
    scale {x=0.8}
    translation {x=-7}
   }
   n68 = "Transform (Modify) Bone - Bone: Dyn_Shorts_r1"!(ComponentPose=@n67, Alpha:float=@n65, Translation="-5,0,0", Rotation="0,0,0", Scale="0.8,1,1") -> (Pose) {
    rotationMode=BMM_Additive
    rotationSpace=BCS_BoneSpace
    scaleMode=BMM_Additive
    scaleSpace=BCS_BoneSpace
    translationMode=BMM_Additive
    translationSpace=BCS_BoneSpace
    alphaBoolBlend {blendInTime=1, blendOption=Linear, blendOutTime=1}
    alphaScaleBiasClamp {bInterpResult=true, interpSpeedDecreasing=6}
    boneToModify {boneName=Dyn_Shorts_r1}
    scale {x=0.8}
    translation {x=-5}
   }
   n69 = "Transform (Modify) Bone - Bone: Dyn_Shorts_r4"!(ComponentPose=@n68, Alpha:float=@n65, Translation="-5,0,0", Rotation="0,0,0", Scale="0.8,1,1") -> (Pose) {
    rotationMode=BMM_Additive
    rotationSpace=BCS_BoneSpace
    scaleMode=BMM_Additive
    scaleSpace=BCS_BoneSpace
    translationMode=BMM_Additive
    translationSpace=BCS_BoneSpace
    alphaBoolBlend {blendInTime=1, blendOption=Linear, blendOutTime=1}
    alphaScaleBiasClamp {bInterpResult=true, interpSpeedDecreasing=6}
    boneToModify {boneName=Dyn_Shorts_r4}
    scale {x=0.8}
    translation {x=-5}
   }
   n70 = Anim Dynamics - Bone: Dyn_Shorts_r2(ComponentPose=@n69, Alpha=1, ExternalForce="0,0,0") {
    alphaBoolBlend {blendOption=Linear}
    boundBone {boneName=Dyn_Shorts_r2}
    physicsBodyDefinitions [{
      boundBone {boneName=Dyn_Shorts_r2}
      boxExtents=<10 10 10>
      constraintSetup {angularLimitsMax=<2 2 0>, angularLimitsMin=<-2 -2 -10>, angularTarget=<0 0 0>}
      localJointOffset=<0 0 0>
     }]
   }
   n71 = Anim Dynamics - Bone: Dyn_Shorts_l2(ComponentPose=@n70, Alpha=1, ExternalForce="0,0,0") {
    alphaBoolBlend {blendOption=Linear}
    boundBone {boneName=Dyn_Shorts_l2}
    physicsBodyDefinitions [{
      boundBone {boneName=Dyn_Shorts_l2}
      boxExtents=<10 10 10>
      constraintSetup {angularLimitsMax=<2 2 0>, angularLimitsMin=<-2 -2 -10>, angularTarget=<0 0 0>}
      localJointOffset=<0 0 0>
     }]
   }
   Output Pose(Result=Component To Local(ComponentPose=@n71)) {layerGroup=None}
   Get Front end head()
  ]}
 EventGraph {nodes [
   Event BlueprintUpdateAnimation()
   > n1 = Sequence!(execute) -> (then_0, then_1, then_2:Exec, then_3:Exec)
   > n2 = Branch(Condition=Get bIsInFrontEnd())
   > Set Front end head(Front end head:double=GetAlphaFromReferenceTransformDelta(
     SkeletalMeshComponent=Get Skeleton() -> (Skeleton:Skeletal Mesh Component Object Reference)
     BoneName=neck_01
     BaseName=spine_05
     DeltaComponentType=QuaternionTwist
     InputRange=(X=-8,Y=-8.5)
     TwistAxis="0,0,1"
    )) -> (Output_Get:double)
   Set Neck_Up(
    execute=@n2.else
    Neck_Up:double=GetAlphaFromReferenceTransformDelta(
     SkeletalMeshComponent=Get Skeleton()
     BoneName=neck_01
     BaseName=spine_05
     DeltaComponentType=QuaternionTwist
     InputRange=(X=-8,Y=20)
     TwistAxis="0,0,1"
    )
   ) -> (Output_Get:double)
   Set Shoulder_SkullNecklace(
    execute=@n1.then_1
    Shoulder_SkullNecklace:double=GetAlphaFromReferenceTransformDelta(
     SkeletalMeshComponent=Get Skeleton()
     BoneName=clavicle_r
     BaseName=spine_05
     DeltaComponentType=EulerX
     InputRange=(X=0.7,Y=-3)
     TwistAxis="0,0,1"
    )
   ) -> (Output_Get:double)
   Set Shorts_RightLeg(
    execute=@n1.then_2
    Shorts_RightLeg:double=GetAlphaFromReferenceTransformDelta(SkeletalMeshComponent=Get Skeleton(), BoneName=thigh_twist_01_r, BaseName=calf_r, InputRange=(X=-24,Y=-117), TwistAxis="0,0,1")
   ) -> (Output_Get:double)
   Set Shorts_LeftLeg(
    execute=@n1.then_3
    Shorts_LeftLeg:double=GetAlphaFromReferenceTransformDelta(SkeletalMeshComponent=Get Skeleton(), BoneName=thigh_twist_01_l, BaseName=calf_l, InputRange=(X=-18,Y=-105), TwistAxis="0,0,1")
   ) -> (Output_Get:double)
   Event BlueprintInitializeAnimation()
   > Set Skeleton(Skeleton:Skeletal Mesh Component Object Reference=GetOwnerSkeletalMeshComponent()) -> (Output_Get:Skeletal Mesh Component Object Reference)
  ]}
]
variables [
 Front end head {type=number}
 Neck_Up {type=number}
 Shorts_LeftLeg {type=number}
 Shorts_RightLeg {type=number}
 Shoulder_SkullNecklace {type=number}
 Skeleton {type=SkeletalMeshComponent}
]
